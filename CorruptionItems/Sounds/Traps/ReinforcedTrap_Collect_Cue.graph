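# Sound cue for collecting a reinforced trap in CorruptionItems, playing a pickup sound in the loot-collect sound class.
# A Random node picks evenly among three ReinforcedWall pickup waves, which then pass through a Modulator for variation.
# The cue plays slightly lowered in pitch and volume with prop-collect attenuation.
# A concurrency override caps simultaneous instances and stops the oldest, alongside an assigned one-per-owner concurrency set.
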SoundCue /CorruptionItems/Sounds/Traps/ReinforcedTrap_Collect_Cue
root=Modulator_0
nodes [
 {id=Modulator_0, children [Random_0]}
 {id=Random_0, children [WavePlayer_2,WavePlayer_3,WavePlayer_4], properties {weights [1,1,1]}}
 {id=WavePlayer_2, properties {soundWaveAssetPtr=&/ReinforcedWall_Pickup_01}}
 {id=WavePlayer_3, properties {soundWaveAssetPtr=&/ReinforcedWall_Pickup_03}}
 {id=WavePlayer_4, properties {soundWaveAssetPtr=&/ReinforcedWall_Pickup_04}}
]
properties {
 attenuationSettings=/Game/Sounds/Attenuation/Fort_Prop_Collect_Att
 bOverrideConcurrency=true
 maxDistance=2097152
 pitchMultiplier=0.9
 priority=0.5
 soundClassObject=&Fort_Loot_Collect
 volumeMultiplier=0.802
 concurrencyOverrides {maxCount=2, resolutionRule=StopOldest}
 concurrencySet [/Game/Sounds/Concurrency_Rules/OneMaximum_Owner_StopLowestPriority]
}
