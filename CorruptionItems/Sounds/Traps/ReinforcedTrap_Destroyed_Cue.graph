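# Sound cue for the destruction of a reinforced trap, built from three reinforced wall destruction waves.
# It picks one wave at random with equal weight, passes it through a modulator at default settings, and adds a short fixed delay.
# It plays slightly pitched down with boosted volume in the impacts sound class, using medium-object demolition attenuation and a concurrency limit for object impacts.
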SoundCue /CorruptionItems/Sounds/Traps/ReinforcedTrap_Destroyed_Cue
root=Delay_0
nodes [
 {id=Delay_0, children [Modulator_0], properties {delayMax=0.07, delayMin=0.07}}
 {id=Modulator_0, children [Random_0]}
 {id=Random_0, children [WavePlayer_0,WavePlayer_1,WavePlayer_2], properties {weights [1,1,1]}}
 {id=WavePlayer_0, properties {soundWaveAssetPtr=&/ReinforcedWall_Destoy_01}}
 {id=WavePlayer_1, properties {soundWaveAssetPtr=&/ReinforcedWall_Destoy_02}}
 {id=WavePlayer_2, properties {soundWaveAssetPtr=&/ReinforcedWall_Destoy_03}}
]
properties {
 attenuationSettings=/Game/Sounds/Attenuation/Fort_Demolition_Destroy_Medium_Sized_Object_Att
 maxDistance=4000
 pitchMultiplier=0.9
 soundClassObject=&Fort_Impacts
 volumeMultiplier=2.5
 concurrencySet [&CONC_Object_Impacts_FourMax]
}
