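# Sound cue for the drop sound of a reinforced trap item, built from ReinforcedWall drop recordings.
# A random node picks one of three equally weighted waves, then a modulator slightly varies pitch and volume to reduce repetition.
# It plays quietly through weapon-drop attenuation and the loot-collect sound class.
# A concurrency override limits it to two instances, stopping the oldest, in place of the assigned one-per-owner concurrency set.
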SoundCue /CorruptionItems/Sounds/Traps/ReinforcedTrap_Drop_Cue
root=Modulator_0
nodes [
 {id=Modulator_0, children [Random_0], properties {pitchMax=1.1, pitchMin=0.9, volumeMax=1.1, volumeMin=0.9}}
 {id=Random_0, children [WavePlayer_0,WavePlayer_2,WavePlayer_3], properties {weights [1,1,1]}}
 {id=WavePlayer_0, properties {soundWaveAssetPtr=&/ReinforcedWall_Drop_01}}
 {id=WavePlayer_2, properties {soundWaveAssetPtr=&/ReinforcedWall_Drop_02}}
 {id=WavePlayer_3, properties {soundWaveAssetPtr=&/ReinforcedWall_Drop_03}}
]
properties {
 attenuationSettings=&Fort_Weapon_Drop_Att
 bOverrideConcurrency=true
 maxDistance=3000
 priority=0.5
 soundClassObject=&Fort_Loot_Collect
 volumeMultiplier=0.398
 concurrencyOverrides {maxCount=2, resolutionRule=StopOldest}
 concurrencySet [/Game/Sounds/Concurrency_Rules/OneMaximum_Owner_StopLowestPriority]
}
