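# Sound cue for the dropping sound of a reinforced trap, built from two ReinforcedWall_Dropping wave variants.
# A Random node picks between the two equally weighted waves, and a Modulator slightly varies pitch and volume on the result.
# It borrows the prop-search attenuation and loot-collect sound class, with a concurrency override that caps simultaneous instances by stopping the oldest.
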SoundCue /CorruptionItems/Sounds/Traps/ReinforcedTrap_Dropping_Cue
root=Modulator_0
nodes [
 {id=Modulator_0, children [Random_1], properties {pitchMax=1.1, pitchMin=0.9, volumeMax=1.1, volumeMin=0.9}}
 {id=Random_1, children [WavePlayer_1,WavePlayer_4], properties {weights [1,1]}}
 {id=WavePlayer_1, properties {soundWaveAssetPtr=&/ReinforcedWall_Dropping_01}}
 {id=WavePlayer_4, properties {soundWaveAssetPtr=&/ReinforcedWall_Dropping_02}}
]
properties {
 attenuationSettings=/Game/Sounds/Attenuation/Fort_Prop_Search_Att
 bOverrideConcurrency=true
 maxDistance=1300
 priority=0.5
 soundClassObject=&Fort_Loot_Collect
 volumeMultiplier=0.597
 concurrencyOverrides {maxCount=2, resolutionRule=StopOldest}
 concurrencySet [/Game/Sounds/Concurrency_Rules/OneMaximum_Owner_StopLowestPriority]
}
