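# Sound cue for equipping the Reinforced Trap, built from Reinforced Wall equip waves and classed as a loot collect sound.
# A mixer blends two pitch-modulated branches, each randomly picking one of three waves, with the EquipLayer branch louder and pitched lower than the main Equip branch.
# It uses prop-collect attenuation and overrides concurrency to allow two instances, stopping the oldest.
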SoundCue /CorruptionItems/Sounds/Traps/ReinforcedTrap_Equip_Cue
root=Mixer_1
nodes [
 {id=Mixer_1, children [Modulator_0,Modulator_1], properties {inputVolume [1.2,0.6]}}
 {id=Modulator_0, children [Random_2], properties {pitchMax=0.9, pitchMin=0.8}}
 {id=Modulator_1, children [Random_0], properties {pitchMax=0.95}}
 {id=Random_0, children [WavePlayer_2,WavePlayer_4,WavePlayer_5], properties {weights [1,1,1]}}
 {id=Random_2, children [WavePlayer_9, WavePlayer_10, WavePlayer_11], properties {weights [1,1,1]}}
 {id=WavePlayer_10, properties {soundWaveAssetPtr=&/ReinforcedWall_EquipLayer_02}}
 {id=WavePlayer_11, properties {soundWaveAssetPtr=&/ReinforcedWall_EquipLayer_03}}
 {id=WavePlayer_2, properties {soundWaveAssetPtr=&/ReinforcedWall_Equip_01}}
 {id=WavePlayer_4, properties {soundWaveAssetPtr=&/ReinforcedWall_Equip_02}}
 {id=WavePlayer_5, properties {soundWaveAssetPtr=&/ReinforcedWall_Equip_03}}
 {id=WavePlayer_9, properties {soundWaveAssetPtr=&/ReinforcedWall_EquipLayer_01}}
]
properties {
 attenuationSettings=/Game/Sounds/Attenuation/Fort_Prop_Collect_Att
 bOverrideConcurrency=true
 maxDistance=2097152
 priority=0.5
 soundClassObject=&Fort_Loot_Collect
 volumeMultiplier=0.239
 concurrencyOverrides {maxCount=2, resolutionRule=StopOldest}
 concurrencySet [/Game/Sounds/Concurrency_Rules/OneMaximum_Owner_StopLowestPriority]
}
