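# Sound cue for reinforced trap impacts, reusing reinforced wall impact waves in the Fort_Impacts sound class.
# An equally weighted Random node picks one of four waves and feeds a Modulator at default settings for slight variation.
# It uses small-object demolition hit attenuation.
# A concurrency override caps simultaneous voices, stopping the oldest and scaling volume, in place of the assigned four-max impact concurrency set.
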SoundCue /CorruptionItems/Sounds/Traps/ReinforcedTrap_Impact_Cue
root=Modulator_4
nodes [
 {id=Modulator_4, children [Random_3]}
 {id=Random_3, children [WavePlayer_7, WavePlayer_8, WavePlayer_11, WavePlayer_12], properties {weights [1,1,1,1]}}
 {id=WavePlayer_11, properties {soundWaveAssetPtr=&/ReinforcedWall_Impact_04}}
 {id=WavePlayer_12, properties {soundWaveAssetPtr=&/ReinforcedWall_Impact_05}}
 {id=WavePlayer_7, properties {soundWaveAssetPtr=&/ReinforcedWall_Impact_02}}
 {id=WavePlayer_8, properties {soundWaveAssetPtr=&/ReinforcedWall_Impact_03}}
]
properties {
 attenuationSettings=/Game/Sounds/Attenuation/Fort_Demolition_Hit_Small_Sized_Object_Att
 bOverrideConcurrency=true
 maxDistance=4000
 soundClassObject=&Fort_Impacts
 volumeMultiplier=0.72
 concurrencyOverrides {maxCount=3, resolutionRule=StopOldest, volumeScale=0.9, volumeScaleAttackTime=0.1}
 concurrencySet [&CONC_Object_Impacts_FourMax]
}
