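# Sound cue that plays when a reinforced wall trap finishes placement.
# A random node picks one of three equally weighted wave variants, and a modulator applies a slight downward pitch variation.
# It uses wall-build attenuation and the Fort_Loot_Collect sound class.
# A concurrency override capping instances at three and stopping the oldest sits alongside a referenced one-per-owner concurrency set.
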SoundCue /CorruptionItems/Sounds/Traps/ReinforcedWall_Place_Complete_Cue
root=Modulator_2
nodes [
 {id=Modulator_2, children [Random_2], properties {pitchMax=1, pitchMin=0.9}}
 {id=Random_2, children [WavePlayer_6,WavePlayer_7,WavePlayer_8], properties {weights [1,1,1]}}
 {id=WavePlayer_6, properties {soundWaveAssetPtr=&/ReinforcedWall_Place_Complete_01}}
 {id=WavePlayer_7, properties {soundWaveAssetPtr=&/ReinforcedWall_Place_Complete_02}}
 {id=WavePlayer_8, properties {soundWaveAssetPtr=&/ReinforcedWall_Place_Complete_03}}
]
properties {
 attenuationSettings=/Game/Sounds/Attenuation/Fort_Build_Wall_Att
 bOverrideConcurrency=true
 maxDistance=2000
 priority=0.5
 soundClassObject=&Fort_Loot_Collect
 volumeMultiplier=0.601
 concurrencyOverrides {maxCount=3, resolutionRule=StopOldest}
 concurrencySet [/Game/Sounds/Concurrency_Rules/OneMaximum_Owner_StopLowestPriority]
}
