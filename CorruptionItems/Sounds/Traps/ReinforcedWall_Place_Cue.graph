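# Sound cue for placing a Reinforced Wall trap, routed through the Fort_Loot_Collect sound class.
# A Random node picks one of three equally weighted place samples, and a Modulator varies its pitch slightly downward.
# The cue adds a lowered base pitch and volume, overrides attenuation with a custom falloff and stereo spread, and overrides concurrency to allow two instances, stopping the oldest.
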SoundCue /CorruptionItems/Sounds/Traps/ReinforcedWall_Place_Cue
root=Modulator_1
nodes [
 {id=Modulator_1, children [Random_1], properties {pitchMax=1, pitchMin=0.9}}
 {id=Random_1, children [WavePlayer_0,WavePlayer_1,WavePlayer_5], properties {weights [1,1,1]}}
 {id=WavePlayer_0, properties {soundWaveAssetPtr=&/ReinforcedWall_Place_01}}
 {id=WavePlayer_1, properties {soundWaveAssetPtr=&/ReinforcedWall_Place_02}}
 {id=WavePlayer_5, properties {soundWaveAssetPtr=&/ReinforcedWall_Place_03}}
]
properties {
 bOverrideAttenuation=true
 bOverrideConcurrency=true
 maxDistance=2000
 pitchMultiplier=0.9
 priority=0.5
 soundClassObject=&Fort_Loot_Collect
 volumeMultiplier=0.677
 attenuationOverrides {falloffDistance=1500, stereoSpread=2000, attenuationShapeExtents {x=500}}
 concurrencyOverrides {maxCount=2, resolutionRule=StopOldest}
 concurrencySet [/Game/Sounds/Concurrency_Rules/OneMaximum_Owner_StopLowestPriority]
}
